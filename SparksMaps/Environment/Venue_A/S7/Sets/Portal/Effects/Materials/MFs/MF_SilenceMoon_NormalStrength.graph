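# MF_SilenceMoon_NormalStrength is a library-exposed material function that adjusts normal map intensity, located with the Venue_A S7 Portal effect materials.
# It linearly interpolates from a flat +Z normal to InNormal by Strength, then normalizes the result.
# InNormal defaults to a preview vector built from tiled, wrapped UVs with Z set to one, and Strength defaults to a halfway blend.
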MaterialFunction /SparksMaps/Environment/Venue_A/S7/Sets/Portal/Effects/Materials/MFs/MF_SilenceMoon_NormalStrength
graph {
 properties {bExposeToLibrary=true}
 nodes [
  n1 = Constant3Vector() {constant=<rgba 0 0 1 1>}
  n2 = TextureCoordinate() {uTiling=4, vTiling=4}
  n3 = FunctionInput(Preview=AppendVector(A=Frac(None=@n2), B=Constant() {r=1})) {bUsePreviewValueAsDefault=true, inputName=InNormal, sortPriority=1}
  n4 = FunctionInput(Preview=Constant() {r=0.5}) {bUsePreviewValueAsDefault=true, inputName=Strength, inputType=FunctionInput_Scalar, sortPriority=2}
  FunctionOutput(None=Normalize(VectorInput=LinearInterpolate(A=@n1, B=@n3, Alpha=@n4)))
 ]
}
