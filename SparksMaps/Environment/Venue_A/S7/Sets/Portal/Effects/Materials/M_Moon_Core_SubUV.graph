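# Translucent, volumetric-directional lit Niagara sprite material for the moon core effect in the Venue_A S7 Portal set.
# SubUV mask and normal textures optionally use motion-vector sampling; selectable channels drive base color, emissive and opacity.
# Particle color modulates emissive and opacity, while a dynamic parameter scales normal strength.
# Opacity-based depth fade and AO tinting apply only on the default shading path, skipping forward and mobile; camera-fade is also computed.
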Material /SparksMaps/Environment/Venue_A/S7/Sets/Portal/Effects/Materials/M_Moon_Core_SubUV
graph {
 properties {
  bUseMaterialAttributes=true
  bUsedWithNiagaraSprites=true
  blendMode=BLEND_Translucent
  pixelDepthOffsetMode=PDOM_Legacy
  translucencyLightingMode=TLM_VolumetricDirectional
  usedShadingModels=Default Lit
 }
 nodes [
  n1 = TextureObjectParameter() {
   group=MotionVector
   parameterName=MotionVector
   samplerType=SAMPLERTYPE_Masks
   texture=/Game/Global/Textures/Default/Blanks/T_Blank_Mask_Yellow_R1G1B0A1
  }
  n2 = TextureObjectParameter() {group=MotionVector, parameterName=NormalMap, samplerType=SAMPLERTYPE_Normal, texture=&FlatNormalBlue}
  n3 = ScalarParameter() {defaultValue=0.005, group=MotionVector, parameterName=MV_Intensity}
  n4 = VectorParameter() {group=MotionVector, parameterName="MV_DirectionSwitchRG(1,-1)", defaultValue=<rgba 1 1 0 1>}
  n5 = StaticBoolParameter() {group=MotionVector, parameterName=MotionVector}
  NamedRerouteDeclaration(None=StaticSwitch(
    True=MaterialFunctionCall!(MotionVector=@n1, SubUVtexture=@n2, Strength=@n3, "DirectionSwitchRG(1.-1)"=@n4.RGB) {materialFunction=/SparksMaps/Environment/Venue_A/S7/Sets/Portal/Effects/Materials/MFs/MF_MOON_SubUVMotionVectors_V2}.Result_RGBA
    False=TextureSample(Tex=@n2).RGB
    Value=@n5
   )) {name=MV_SampledNormalTexture, nodeColor=<rgba 0.0153 1 0 1>}
  n6 = ChannelMaskParameter(None=NamedRerouteUsage()) {group=Opacity, maskChannel=Green, parameterName=OpacityChannel, defaultValue=<rgba 0 1 0 0>}
  n7 = MaterialFunctionCall!(
   InAlpha=@n6.R
   OpacityBasedFade_DistanceA=ScalarParameter() {defaultValue=50, group=Opacity, parameterName=OpacityBasedFade_DistanceA}
   OpacityBasedFade_DistanceB=ScalarParameter() {defaultValue=5, group=Opacity, parameterName=OpacityBasedFade_DistanceB}
  ) {materialFunction=/SparksMaps/Environment/Venue_A/S7/Sets/Portal/Effects/Materials/MFs/MF_SilenceMoon_OpacityBasedDepthFade}
  NamedRerouteDeclaration(None=@n7.FadeValue) {name=DepthFadeValue, nodeColor=<rgba 0 1 0.5313 1>}
  n8 = TextureObjectParameter() {group=MotionVector, parameterName=Masks, samplerType=SAMPLERTYPE_Masks, texture=/Game/Global/Textures/Default/Blanks/AllRed}
  NamedRerouteDeclaration(None=StaticSwitch(
    True=MaterialFunctionCall(MotionVector=@n1, SubUVtexture=@n8, Strength=@n3, "DirectionSwitchRG(1.-1)"=@n4.RGB) {materialFunction=&MF_MOON_SubUVMotionVectors_V2}.Result_RGBA
    False=TextureSample(Tex=@n8).RGBA
    Value=@n5
   )) {name=MV_SampledTexture, nodeColor=<rgba 0.0153 1 0 1>}
  NamedRerouteDeclaration(None=MaterialFunctionCall(
    Fade Length=ScalarParameter() {defaultValue=300, parameterName=Camfade_Lemgth}
    Fade Offset=ScalarParameter() {parameterName=Camfade_Offset}
    For Vertex Shader=StaticBool() {value=true}
   ) {materialFunction=&CameraDepthFade}.Result) {name=CamFade, nodeColor=<rgba 1 0 0.9613 1>}
  n9 = VectorParameter() {group=BaseColor, parameterName=AOColor, defaultValue=<rgba 0.15 0.15 0.15 1>}
  n10 = LinearInterpolate(
   A=VectorParameter() {group=BaseColor, parameterName=BaseColorTintA, defaultValue=<rgba 0.005 0.005 0.005 1>}.RGB
   B=VectorParameter() {group=BaseColor, parameterName=BaseColorTintB, defaultValue=<rgba 0.5 0.5 0.5 1>}.RGB
   Alpha=ChannelMaskParameter(None=NamedRerouteUsage()) {group=BaseColor, parameterName=BaseColorChannel}.R
  )
  n11 = NamedRerouteUsage()
  NamedRerouteDeclaration(None=ShadingPathSwitch(Default=LinearInterpolate(A=@n9.RGB, B=@n10, Alpha=@n11), Forward=@n10, Mobile=@n10)) {name=BaseColor, nodeColor=<rgba 0 0 0 1>}
  n12 = VectorParameter() {group=Emissive, parameterName=EmissiveColorTintA, defaultValue=<rgba 0 0 0 1>}
  n13 = VectorParameter() {group=Emissive, parameterName=EmissiveColorTintB, defaultValue=<rgba 0 0.5 1 1>}
  n14 = ChannelMaskParameter(None=NamedRerouteUsage()) {group=Emissive, parameterName=EmissiveChannel}
  NamedRerouteDeclaration(None=Multiply(A=ParticleColor().RGB, B=LinearInterpolate(A=@n12.RGB, B=@n13.RGB, Alpha=@n14.R))) {name=Emissive, nodeColor=<rgba 0 0 0 1>}
  NamedRerouteDeclaration(None=MaterialFunctionCall!(InNormal=NamedRerouteUsage(), Strength=DynamicParameter() {paramNames [Param1, Param2, Param3, NormalStrength]}.NormalStrength) {materialFunction=/SparksMaps/Environment/Venue_A/S7/Sets/Portal/Effects/Materials/MFs/MF_SilenceMoon_NormalStrength}.Result) {name=Normal, nodeColor=<rgba 0 0 0 1>}
  n15 = ShadingPathSwitch(Default=@n7.FinalAlpha, Forward=@n6.R, Mobile=@n6.R)
  n16 = NamedRerouteUsage()
  NamedRerouteDeclaration(None=Multiply(A=ParticleColor().A, B=Multiply(A=@n15, B=@n16))) {name=Opacity, nodeColor=<rgba 0 0 0 1>}
  SetMaterialAttributes!(
   MaterialAttributes
   Opacity=NamedRerouteUsage()
   Base Color=NamedRerouteUsage()
   Emissive Color=NamedRerouteUsage()
   Normal=NamedRerouteUsage()
  ) {attributeSetTypes [
    #1
    #2
    #3
    #4
   ]}
 ]
}
