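# Translucent, two-sided unlit material for the outer cone of the Sparks Venue A S7 portal effect, usable on Niagara mesh particles.
# Two panning samples of one texture, distorted by scrolling noise, multiply into a pattern driving opacity and emissive.
# Color comes from the stage's tertiary color or two fallback colors via static switch.
# A V-based center mask boosts both outputs, while vertex color, optional fresnel and non-mobile depth fade shape opacity.
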Material /SparksMaps/Environment/Venue_A/S7/Sets/Portal/Effects/Materials/M_Sparks_Portal_OuterCone
graph {
 properties {
  bCastRayTracedShadows=false
  bUsedWithNanite=true
  bUsedWithNiagaraMeshParticles=true
  blendMode=BLEND_Translucent
  pixelDepthOffsetMode=PDOM_Legacy
  shadingModel=MSM_Unlit
  twoSided=true
  usedShadingModels=Unlit
 }
 nodes [
  n1 = VectorParameter() {group=Texture, parameterName=Tex01UV_Scale_PanSpeed, defaultValue=<rgba 10 10 -0.15 0.035>}
  n2 = TextureCoordinate()
  n3 = MaterialFunctionCall(Float2=@n2) {materialFunction=&BreakOutFloat2Components}
  n4 = Multiply(A=AppendVector(A=@n3.R, B=Add(A=@n3.R, B=@n3.R)), B=Constant2Vector() {g=1, r=-1})
  n5 = VectorParameter() {group=Distortion, parameterName=NoiseUV_Scale_PanSpeed, defaultValue=<rgba 2 1 0.1 -0.1>}
  n6 = Panner(Coordinate=Multiply(A=@n2, B=ComponentMask(None=@n5.RGB) {g=true, r=true}), Speed=AppendVector(A=@n5.B, B=@n5.B))
  n7 = Multiply(
   A=ComponentMask(None=TextureSample(UVs=@n6) {samplerType=SAMPLERTYPE_LinearColor, texture=/SparksMaps/Environment/Venue_A/S7/Sets/Portal/Effects/T_FX_Tile_0138_Moon}.RGB) {g=true, r=true}
   B=Multiply(
    A=ScalarParameter() {defaultValue=0.1, group=Distortion, parameterName=NoiseIntensity}
    B=TextureSample() {samplerType=SAMPLERTYPE_LinearColor, texture=&T_FX_Tile_0138_Moon}.B
   )
  )
  n8 = Multiply(A=ComponentMask(None=@n1.RGB) {g=true, r=true}, B=Add(A=@n4, B=@n7))
  n9 = AppendVector(A=@n1.B, B=@n1.B)
  n10 = TextureObjectParameter() {
   group=Texture
   parameterName=Texture
   samplerType=SAMPLERTYPE_LinearColor
   texture=/SparksMaps/Environment/Venue_A/S7/Sets/Portal/Effects/Tile_output_8
  }
  n11 = VectorParameter() {group=Texture, parameterName=Tex02UVs_Scale_PanSpeed, defaultValue=<rgba 4 4 -0.1 0.05>}
  n12 = Multiply(A=Add(A=@n4, B=@n7), B=ComponentMask(None=@n11.RGB) {g=true, r=true})
  n13 = AppendVector(A=@n11.B, B=@n11.B)
  n14 = Multiply(
   A=TextureSample(UVs=Panner(Coordinate=@n8, Speed=@n9), Tex=@n10) {samplerType=SAMPLERTYPE_LinearColor}.G
   B=TextureSample(UVs=Panner(Coordinate=@n12, Speed=@n13), Tex=@n10) {samplerType=SAMPLERTYPE_LinearColor}.G
  )
  n15 = ScalarParameter() {defaultValue=2, group=Texture, parameterName=TextureBrightness}
  n16 = Saturate(None=Power(Base=Multiply(A=@n14, B=@n15), Exp=ScalarParameter() {defaultValue=1.6, group=Texture, parameterName=TexturePow}))
  n17 = Add(A=@n16, B=Multiply(A=NamedRerouteUsage(), B=ScalarParameter() {defaultValue=3, parameterName=ShadowAdd}))
  n18 = ShadingPathSwitch(
   Default=DepthFade(Opacity=@n17, FadeDistance=ScalarParameter() {defaultValue=50, group=Alpha, parameterName=DepthFade_Distance})
   Mobile=@n17
  )
  n19 = VertexColor()
  n20 = ScalarParameter() {defaultValue=1, group=Alpha, parameterName=VertColor_Pow}
  n21 = Abs(None=DotProduct(A=VertexNormalWS(), B=CameraVectorWS()))
  n22 = Multiply(
   A=LinearInterpolate(B=Power(Base=@n19.R, Exp=@n20), Alpha=ScalarParameter() {defaultValue=1, group=Alpha, parameterName=VertexColorAlpha_Intensity}) {constA=1}
   B=LinearInterpolate(A=Constant() {r=1}, B=Multiply(A=@n21), Alpha=ScalarParameter() {group=Alpha, parameterName=AlphaFresnel_Intensity})
  )
  LinearInterpolate(
   A=Multiply(A=Multiply(A=@n18, B=@n22), B=ScalarParameter() {defaultValue=0.75, group=Alpha, parameterName=BaseOpacity})
   Alpha=NamedRerouteUsage()
  )
  n23 = MaterialFunctionCall!(CustomUPosition) {materialFunction=/SparksStageShow/Systems/StageManager/FXManager/MaterialFunctions/MF_MojoStageColors}
  n24 = Add(B=@n16) {constA=0.75}
  n25 = StaticSwitchParameter(
   True=LinearInterpolate(A=@n23.Tertiary, B=@n23.Tertiary, Alpha=Saturate(None=@n24))
   False=LinearInterpolate(
    A=VectorParameter() {parameterName=Color02, defaultValue=<rgba 1 0.0264 0 1>}.RGB
    B=VectorParameter() {parameterName=Color01, defaultValue=<rgba 1 0 0 0>}.RGB
    Alpha=@n16
   )
  ) {defaultValue=true, group=Texture, parameterName=Use Stage Colors}
  n26 = Multiply(A=@n25, B=@n17)
  n27 = ScalarParameter() {defaultValue=2, parameterName=TextureEmissiveAmount}
  MaterialFunctionCall(Emissive=Multiply(A=LinearInterpolate(A=@n25, B=@n26, Alpha=@n27), B=ScalarParameter() {defaultValue=0.2, parameterName=MasterBrightness})) {materialFunction=&MF_FX_EmissiveAdjustments}
  Multiply(A=TextureCoordinate(), B=ScalarParameter() {defaultValue=5})
  n28 = ComponentMask(None=TextureCoordinate()) {g=true}
  n29 = Saturate(None=Multiply(A=OneMinus(None=@n28)) {constB=10})
  n30 = Add(A=Multiply(A=OneMinus(None=@n29), B=@n29), B=ScalarParameter() {defaultValue=-0.025, parameterName=CenterMaskUVOffset})
  n31 = Power(Base=Saturate(None=Multiply(A=@n30) {constB=2}), Exp=ScalarParameter() {defaultValue=1, parameterName=CenterMaskPow}) {constExponent=1}
  n32 = ScalarParameter() {defaultValue=1, parameterName=CenterMaskOpacity}
  NamedRerouteDeclaration(None=Saturate(None=LinearInterpolate(B=@n31, Alpha=@n32))) {name=CenterMask, nodeColor=<rgba 1 0 0.7266 1>}
 ]
}
