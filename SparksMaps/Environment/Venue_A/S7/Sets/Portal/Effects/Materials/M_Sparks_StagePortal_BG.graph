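# Unlit, two-sided material for the Sparks stage portal background, usable on Nanite meshes and Niagara sprites.
# A tech tile texture, sampled with radial, panned UVs, modulates the hardness of inverted sphere masks that shape the glow.
# Bump-offset and tilted UV reroutes supply fake depth and skew.
# Stage colors or a fixed color tint the result, with brightness pulsed by what appears to be audio amplitude, before an emissive adjustment function.
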Material /SparksMaps/Environment/Venue_A/S7/Sets/Portal/Effects/Materials/M_Sparks_StagePortal_BG
graph {
 properties {
  bUsedWithNanite=true
  bUsedWithNiagaraSprites=true
  pixelDepthOffsetMode=PDOM_Legacy
  shadingModel=MSM_Unlit
  twoSided=true
  usedShadingModels=Unlit
 }
 nodes [
  n1 = CollectionParameter() {collection=/FMAudioReactiveData/AudioAnalysis/MPC_Sparks_Audio_Parameters, parameterName=Amplitude}
  CurveAtlasRowParameter(CurveTime=@n1) {
   atlas=/SparksMaps/Effects/Screens/Curves/CurveAtlas_Audio_Beat_A
   curve=/SparksMaps/Effects/Screens/Curves/Curve_Audio_Beat_Remap_001
   parameterName=ReactivityCurve
  }
  LinearInterpolate(A=ParticleColor().RGB, Alpha=ScalarParameter() {defaultValue=1, parameterName=UseMatBeatReact?}) {constA=1}
  n2 = LinearInterpolate(
   A=ScalarParameter() {defaultValue=0.85, parameterName=Param_13}
   B=ScalarParameter() {defaultValue=2, parameterName=Param_9}
   Alpha=NamedRerouteUsage()
  ) {constB=3}
  n3 = ScalarParameter() {defaultValue=1, parameterName=Pow}
  n4 = NamedRerouteUsage()
  n5 = Constant() {r=0.5}
  n6 = ScalarParameter() {defaultValue=0.676, parameterName=Param_2}
  n7 = ScalarParameter() {defaultValue=-0.2452, parameterName=Param_4}
  n8 = Add(B=OneMinus(None=SphereMask(A=@n4, B=@n5, Radius=@n6, Hardness=@n7) {hardnessPercent=50}))
  n9 = TextureCoordinate()
  n10 = Constant() {r=0.5}
  n11 = ScalarParameter() {defaultValue=0.5, parameterName=Param_7}
  n12 = TextureSampleParameter2D!(UVs=NamedRerouteUsage(), Level=ScalarParameter() {defaultValue=1.5, parameterName=Param_6}, Apply View MipBias) {
   mipValueMode=TMVM_MipLevel
   parameterName=Param
   samplerType=SAMPLERTYPE_LinearColor
   texture=/SparksMaps/Environment/Venue_A/S7/Sets/Portal/Effects/Tile_Tech_02
  }
  n13 = LinearInterpolate(
   A=ScalarParameter() {defaultValue=0.2963, parameterName=Param_3}
   B=ScalarParameter() {defaultValue=0.8602, parameterName=Param_11}
   Alpha=SquareRoot(None=@n12.G)
  )
  n14 = NamedRerouteUsage()
  n15 = Constant() {r=0.5}
  n16 = ScalarParameter() {defaultValue=1.1629, parameterName=Param_1}
  n17 = Max(
   A=OneMinus(None=SphereMask(A=@n9, B=@n10, Radius=@n11, Hardness=@n13) {hardnessPercent=50})
   B=OneMinus(None=SphereMask(A=@n14, B=@n15, Radius=@n16, Hardness=@n13) {hardnessPercent=50})
  )
  n18 = MaterialFunctionCall!(CustomUPosition) {materialFunction=/SparksStageShow/Systems/StageManager/FXManager/MaterialFunctions/MF_MojoStageColors}
  n19 = VectorParameter() {parameterName=Color, defaultValue=<rgba 0 1 1 1>}
  n20 = Multiply(
   A=Power(Base=Multiply(A=@n17, B=@n8), Exp=ScalarParameter() {defaultValue=2, parameterName=Param_8})
   B=Multiply(
    A=StaticSwitchParameter(True=@n18.Primary, False=@n19.RGB) {defaultValue=true, parameterName=Use Stage Colors}
    B=ScalarParameter() {defaultValue=2, parameterName=Brightness}
   )
  )
  n21 = Multiply(A=@n18.Tertiary) {constB=0.01}
  MaterialFunctionCall(Emissive=Multiply(A=Power(Base=@n2, Exp=@n3), B=Add(A=@n20, B=@n21) {constB=0})) {materialFunction=&MF_FX_EmissiveAdjustments}
  Multiply(
   A=ScalarParameter() {defaultValue=25, parameterName=Param_5}
   B=MaterialFunctionCall() {materialFunction=&DitherTemporalAA_Fortnite}.Result
  )
  NamedRerouteDeclaration(None=BumpOffset(Coordinate=TextureCoordinate(), Height=ScalarParameter() {defaultValue=-12.5039}) {referencePlane=0}) {name=BumpUV, nodeColor=<rgba 0.6406 0 1 1>}
  n22 = Add(A=NamedRerouteUsage()) {constB=-0.5}
  n23 = VectorParameter() {parameterName=Main_UV, defaultValue=<rgba 3 2.101 -0.1307 0.248>}
  NamedRerouteDeclaration(None=MaterialFunctionCall!(
    ClampTexture
    UV=MaterialFunctionCall(Vector or UVs=@n22) {materialFunction=&VectorToRadialValue}.Radial Coordinates
    Tex
    InRGB=@n23.RGB
    InA=@n23.RGB
    Time
    TimeMult
   ) {materialFunction=/SparksMaps/Environment/Venue_A/S7/Sets/Portal/Effects/Materials/MFs/MF_MOON_TexPanner_V2}.PannerUV) {name=MainUV, nodeColor=<rgba 0 0.4049 1 1>}
  n24 = MaterialFunctionCall(Float2=NamedRerouteUsage()) {materialFunction=&BreakOutFloat2Components}
  n25 = Multiply(A=@n24.R, B=ScalarParameter() {defaultValue=2, parameterName=UV_Tilt})
  NamedRerouteDeclaration(None=MaterialFunctionCall(X=@n24.R, Y=Add(A=@n24.G, B=@n25)) {materialFunction=&MakeFloat2}.Result) {name=TiltedUV, nodeColor=<rgba 0 0.3848 1 1>}
  NamedRerouteDeclaration(None=Power(Base=SmoothStep(Value=@n1), Exp=ScalarParameter() {defaultValue=1, parameterName=Amplitude_Pow})) {name=React_Beat, nodeColor=<rgba 0.2362 1 0 1>}
 ]
}
