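# Unlit, Nanite-enabled emissive material for the StagePortal set piece in the Sparks Venue_A S7 stage environment.
# A static switch, on by default, selects the primary color from MF_MojoStageColors or a fallback Color parameter.
# The selected color is multiplied by a Brightness scalar and fed into the shared MF_FX_EmissiveAdjustments function.
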Material /SparksMaps/Environment/Venue_A/S7/Sets/Portal/Effects/Materials/M_Sparks_StagePortal_Emissive
graph {
 properties {bAutomaticallySetUsageInEditor=false, bUsedWithNanite=true, shadingModel=MSM_Unlit, usedShadingModels=Unlit}
 nodes [
  n1 = MaterialFunctionCall!(CustomUPosition) {materialFunction=/SparksStageShow/Systems/StageManager/FXManager/MaterialFunctions/MF_MojoStageColors}
  n2 = VectorParameter() {parameterName=Color, defaultValue=<rgba 0 1 1 1>}
  MaterialFunctionCall(Emissive=Multiply(
    A=StaticSwitchParameter(True=@n1.Primary, False=@n2.RGB) {defaultValue=true, parameterName=Use Stage Colors}
    B=ScalarParameter() {defaultValue=20, parameterName=Brightness}
   )) {materialFunction=&MF_FX_EmissiveAdjustments}
 ]
}
