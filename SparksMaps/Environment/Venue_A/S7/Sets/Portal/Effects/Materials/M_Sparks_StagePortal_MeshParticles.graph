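# Unlit material for Niagara mesh particles in the Sparks stage portal effect.
# Particle color blends toward an inverted-Fresnel-weighted version by Fresnel_Strength, then MasterBrightness scales it into MF_FX_EmissiveAdjustments.
# Separate branches, which do not feed that chain, combine two bump-offset snowball noise samples and switch between Mojo stage colors and a Color parameter.
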Material /SparksMaps/Environment/Venue_A/S7/Sets/Portal/Effects/Materials/M_Sparks_StagePortal_MeshParticles
graph {
 properties {bAutomaticallySetUsageInEditor=false, bUsedWithNiagaraMeshParticles=true, shadingModel=MSM_Unlit, usedShadingModels=Unlit}
 nodes [
  n1 = Transform(None=CameraVectorWS()) {transformSourceType=TRANSFORMSOURCE_World, transformType=TRANSFORM_Tangent}
  n2 = Multiply(A=TextureCoordinate() {uTiling=2, vTiling=2}, B=ScalarParameter() {defaultValue=1, parameterName=Tile})
  n3 = Constant()
  n4 = TextureSample(UVs=BumpOffset(Coordinate=@n2, Height=@n3)) {texture=/Game/Abilities/Toys/Snowball/Textures/T_Snowball_Noise_01}
  n5 = TextureSample(UVs=BumpOffset(Coordinate=@n2, Height=@n3) {referencePlane=0.2}) {texture=&T_Snowball_Noise_01}
  n6 = Add(A=@n4.R, B=@n5.R)
  Clamp(None=Multiply(A=ComponentMask(None=@n1) {b=true}, B=@n6))
  n7 = ParticleColor()
  n8 = ScalarParameter() {defaultValue=1, parameterName=FresnelExponent}
  n9 = Multiply(A=@n7.RGB, B=OneMinus(None=Fresnel(ExponentIn=@n8)))
  n10 = ScalarParameter() {defaultValue=1, parameterName=Fresnel_Strength}
  MaterialFunctionCall(Emissive=Multiply(A=ScalarParameter() {defaultValue=1, parameterName=MasterBrightness}, B=LinearInterpolate(A=@n7.RGB, B=@n9, Alpha=@n10))) {materialFunction=&MF_FX_EmissiveAdjustments}
  Multiply(A=@n6, B=Multiply(A=@n4.R, B=@n5.R))
  Multiply(
   A=StaticSwitchParameter(
    True=MaterialFunctionCall!(CustomUPosition) {materialFunction=/SparksStageShow/Systems/StageManager/FXManager/MaterialFunctions/MF_MojoStageColors}.Primary
    False=VectorParameter() {parameterName=Color, defaultValue=<rgba 0 1 1 1>}.RGB
   ) {defaultValue=true, parameterName=Use Stage Colors}
   B=ScalarParameter() {defaultValue=5, parameterName=Brightness}
  )
 ]
}
