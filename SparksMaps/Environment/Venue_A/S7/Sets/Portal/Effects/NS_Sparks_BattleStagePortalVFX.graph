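# Niagara effect for the Sparks battle stage portal: rising mesh boxes around a ring, plus ground smoke.
# Cheap_Boxes (CPU) and denser GPU_Boxes001 share near-identical stacks, with cone velocity, upward gravity, drag and curve-driven scale.
# Boxes blend from StageShowPrimary to StageShowTertiary in the NPC_Sparks parameter collection; their audio-beat scale module is disabled.
# SmokeGround spawns sparse, animated sub-image sprites on a smaller ring, fading from the primary color to dark gray.
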NiagaraSystem /SparksMaps/Environment/Venue_A/S7/Sets/Portal/Effects/NS_Sparks_BattleStagePortalVFX
dependencies {
 usedDataInterfaces [&NiagaraDataInterfaceCurve, &NiagaraDataInterfaceVectorCurve]
 usedDynamicInputs []
 usedModules [&SystemState]
}
emitters {
 Cheap_Boxes {
  emitterUpdateScript [
   &EmitterState(Life Cycle Mode=::NewEnumerator1, Loop Duration=2, UseLoopDelay=false)
   &SpawnRate(SpawnRate=25, Use Spawn Probability=false)
  ]
  particleSpawnScript [
   &InitializeParticle(
    Color Mode=::NewEnumerator0
    Lifetime Max=3
    Lifetime Mode=::NewEnumerator1
    Mesh Scale Mode=::NewEnumerator3
    Mesh Uniform Scale=&RandomRangeFloat(Minimum=45, Maximum=60)
    Position Offset=<0 0 200>
    Sprite Rotation Mode=::NewEnumerator2
    Sprite Size Mode=::NewEnumerator1
    Uniform Sprite Size Max=500
    Uniform Sprite Size Min=500
    UsePositionOffset=true
    Write Material Random=true
   )
   &ShapeLocation(
    Disc Coverage=0.3095
    MaskSpawnGroup=false
    Override Seed=false
    Ring Radius=550
    Rotation Axis=<0 0 0>
    Rotation Mode=::NewEnumerator1
    Shape Primitive=::NewEnumerator4
    UseOwnerScale=true
   ) {module=ShapeLocation001}
   &AddVelocity(
    Cone Angle=60
    Cone Axis=<0 0 1>
    Distribution Along Cone Axis=0.25
    Module.UseConeFalloff=true
    Module.UseVelDistribution=true
    Override Seed=false
    Speed Falloff From Cone Axis=0.5
    Velocity Mode=::NewEnumerator2
    Velocity Speed=&RandomRangeFloat(Minimum=-50, Maximum=150)
   )
  ]
  particleUpdateScript [
   &ParticleState()
   &ScaleMeshSize(Scale Factor=&Lerp_Vector(
     A=<0.5 0.5 0.5>
     B=&RandomRangeVector(Minimum=<0.8 0.8 0.8>, Maximum=<0.35 0.35 1.5>)
     Alpha=&FloatFromCurve(
      FloatCurve=DataInterface() {propertyValues {Curve {keys [
          {arriveTangentWeight=0.0434, interpMode=RCIM_Cubic, leaveTangentWeight=0.0868, tangentMode=RCTM_User, time=0, value=1}
          {arriveTangentWeight=0.1294, interpMode=RCIM_Cubic, leaveTangentWeight=0.1294, tangentMode=RCTM_User, time=0.05, value=1}
          {arriveTangent=-1.0526, interpMode=RCIM_Cubic, leaveTangent=-1.0526, time=0.3, value=0.35}
          <at 1 0 RCIM_Cubic>
         ]}}}
      CurveIndex=$NPC.NPC_Sparks_Audio_Parameters001.BeatLifetime
     )
    )) {enabled=false, module=ScaleMeshSize002}
   &ScaleMeshSize(Scale Factor=&Lerp_Vector(
     A=<0.5 0.5 0.5>
     B=&RandomRangeVector(Minimum=<0.8 0.8 0.8>, Maximum=<0.5 0.5 1.25>)
     Alpha=&FloatFromCurve(FloatCurve=DataInterface() {propertyValues {Curve {keys [
          {arriveTangentWeight=0.0434, interpMode=RCIM_Cubic, leaveTangentWeight=0.0868, tangentMode=RCTM_User, time=0, value=1}
          {arriveTangentWeight=0.1294, interpMode=RCIM_Cubic, leaveTangentWeight=0.1294, tangentMode=RCTM_User, time=0.05, value=1}
          {arriveTangent=-1.0526, interpMode=RCIM_Cubic, leaveTangent=-1.0526, time=0.3, value=0.35}
          <at 1 0 RCIM_Cubic>
         ]}}})
    )) {module=ScaleMeshSize001}
   &ScaleMeshSize(Scale Factor=&VectorFromFloat(Value=&FloatFromCurve(
      FloatCurve=DataInterface() {propertyValues {Curve {keys [
          {arriveTangent=6.2606, interpMode=RCIM_Cubic, leaveTangent=6.2606, tangentMode=RCTM_User, time=0, value=0}
          {arriveTangent=-0.0222, interpMode=RCIM_Cubic, leaveTangent=-0.0222, tangentMode=RCTM_User, time=0.2, value=0.95}
          {arriveTangent=-1, interpMode=RCIM_Cubic, leaveTangent=-1, time=0.4, value=0.3}
          {arriveTangent=0.7255, interpMode=RCIM_Cubic, leaveTangent=0.7255, tangentMode=RCTM_User, time=0.6, value=0.55}
          {arriveTangent=-6.0844, interpMode=RCIM_Cubic, leaveTangent=-6.0844, tangentMode=RCTM_User, time=1, value=0}
         ]}}}
      Scale Curve=1.5
     )))
   &GravityForce(Gravity=<0 0 60>)
   &Drag(Drag=0.5, Use Linear Drag=true, Use Rotational Drag=false)
   &ScaleColor(
    Scale Alpha=&FloatFromCurve(FloatCurve=DataInterface() {propertyValues {Curve {keys [<at -0.0046 1.0113 RCIM_Cubic>,<at 1.0011 0.002 RCIM_Cubic>]}}})
    Scale RGB=&VectorFromCurve(
     VectorCurve=DataInterface() {propertyValues {
       XCurve {keys [
         {arriveTangent=-1.8006, interpMode=RCIM_Cubic, leaveTangent=-1.8006, tangentMode=RCTM_User, time=0, value=1}
         {interpMode=RCIM_Cubic, leaveTangent=-0.8583, time=1, value=0.35}
        ]}
       YCurve {keys [
         {arriveTangent=-1.8006, interpMode=RCIM_Cubic, leaveTangent=-1.8006, tangentMode=RCTM_User, time=0, value=1}
         {interpMode=RCIM_Cubic, leaveTangent=-0.8583, time=1, value=0.35}
        ]}
       ZCurve {keys [
         {arriveTangent=-1.8006, interpMode=RCIM_Cubic, leaveTangent=-1.8006, tangentMode=RCTM_User, time=0, value=1}
         {interpMode=RCIM_Cubic, leaveTangent=-0.8583, time=1, value=0.35}
        ]}
      }}
     Scale Curve=&Multiply_VectorByFloat(Vector=<1 1 1>, Float=1.5)
    )
    ScaleA=false
    ScaleRGB=true
   )
   &ScaleColor(
    Color Value To Scale=$Particles.Color
    Scale RGB=&Lerp_Vector(
     A=&MakeVectorFromLinearColorRGB(Color=&Multiply_LinearColorByFloat(Linear Color=$NPC.NPC_Sparks.StageShowPrimary, Float=6))
     B=&MakeVectorFromLinearColorRGB(Color=&Multiply_LinearColorByFloat(Linear Color=$NPC.NPC_Sparks.StageShowTertiary, Float=1))
     Alpha=$StackContext.NormalizedAge
    )
    ScaleA=false
    ScaleRGB=true
   ) {module=ScaleColor002}
   &CurlNoiseForce(Bias Noise=false, Noise Frequency=500, Noise Strength=1000, Pan Noise=true, Pan Noise Field=<0 0 100>, Randomization Vector:Vector3f=<0 0 0>) {enabled=false}
   &VortexForce(UseOriginOffset=false) {enabled=false}
   &SolveForcesAndVelocity(Clamp Velocity=false, Limit Acceleration=false)
  ]
  rendererClasses [&NiagaraMeshRendererProperties]
  renderers [{rendererClass=&NiagaraMeshRendererProperties, rendererIndex=0}]
 }
 GPU_Boxes001 {
  simTarget=GPUComputeSim
  emitterUpdateScript [
   &EmitterState(Life Cycle Mode=::NewEnumerator1, Loop Duration=2, UseLoopDelay=false)
   &SpawnRate(SpawnRate=60, Use Spawn Probability=false)
  ]
  particleSpawnScript [
   &InitializeParticle(
    Color Mode=::NewEnumerator0
    Lifetime Max=3
    Lifetime Mode=::NewEnumerator1
    Mesh Scale Mode=::NewEnumerator3
    Mesh Uniform Scale=&RandomRangeFloat(Minimum=35, Maximum=60)
    Position Offset=<0 0 200>
    Sprite Rotation Mode=::NewEnumerator2
    Sprite Size Mode=::NewEnumerator1
    Uniform Sprite Size Max=500
    Uniform Sprite Size Min=500
    UsePositionOffset=true
    Write Material Random=true
   )
   &ShapeLocation(
    Disc Coverage=0.3095
    MaskSpawnGroup=false
    Override Seed=false
    Ring Radius=550
    Rotation Axis=<0 0 0>
    Rotation Mode=::NewEnumerator1
    Shape Primitive=::NewEnumerator4
    UseOwnerScale=true
   ) {module=ShapeLocation001}
   &AddVelocity(
    Cone Angle=60
    Cone Axis=<0 0 1>
    Distribution Along Cone Axis=0.25
    Module.UseConeFalloff=true
    Module.UseVelDistribution=true
    Override Seed=false
    Speed Falloff From Cone Axis=0.5
    Velocity Mode=::NewEnumerator2
    Velocity Speed=&RandomRangeFloat(Minimum=-50, Maximum=150)
   )
  ]
  particleUpdateScript [
   &ParticleState()
   &ScaleMeshSize(Scale Factor=&Lerp_Vector(
     A=<0.5 0.5 0.5>
     B=&RandomRangeVector(Minimum=<0.8 0.8 0.8>, Maximum=<0.35 0.35 1.5>)
     Alpha=&FloatFromCurve(
      FloatCurve=DataInterface() {propertyValues {Curve {keys [
          {arriveTangentWeight=0.0434, interpMode=RCIM_Cubic, leaveTangentWeight=0.0868, tangentMode=RCTM_User, time=0, value=1}
          {arriveTangentWeight=0.1294, interpMode=RCIM_Cubic, leaveTangentWeight=0.1294, tangentMode=RCTM_User, time=0.05, value=1}
          {arriveTangent=-1.0526, interpMode=RCIM_Cubic, leaveTangent=-1.0526, time=0.3, value=0.35}
          <at 1 0 RCIM_Cubic>
         ]}}}
      CurveIndex=$NPC.NPC_Sparks_Audio_Parameters001.BeatLifetime
     )
    )) {enabled=false, module=ScaleMeshSize002}
   &ScaleMeshSize(Scale Factor=&Lerp_Vector(
     A=<0.5 0.5 0.5>
     B=&RandomRangeVector(Minimum=<0.8 0.8 0.8>, Maximum=<0.5 0.5 1.25>)
     Alpha=&FloatFromCurve(FloatCurve=DataInterface() {propertyValues {Curve {keys [
          {arriveTangentWeight=0.0434, interpMode=RCIM_Cubic, leaveTangentWeight=0.0868, tangentMode=RCTM_User, time=0, value=1}
          {arriveTangentWeight=0.1294, interpMode=RCIM_Cubic, leaveTangentWeight=0.1294, tangentMode=RCTM_User, time=0.05, value=1}
          {arriveTangent=-1.0526, interpMode=RCIM_Cubic, leaveTangent=-1.0526, time=0.3, value=0.35}
          <at 1 0 RCIM_Cubic>
         ]}}})
    )) {module=ScaleMeshSize001}
   &ScaleMeshSize(Scale Factor=&VectorFromFloat(Value=&FloatFromCurve(
      FloatCurve=DataInterface() {propertyValues {Curve {keys [
          {arriveTangent=6.2606, interpMode=RCIM_Cubic, leaveTangent=6.2606, tangentMode=RCTM_User, time=0, value=0}
          {arriveTangent=-0.0222, interpMode=RCIM_Cubic, leaveTangent=-0.0222, tangentMode=RCTM_User, time=0.2, value=0.95}
          {arriveTangent=-1.125, interpMode=RCIM_Cubic, leaveTangent=-1.125, time=0.4, value=0.3}
          {arriveTangent=-1.7004, interpMode=RCIM_Cubic, leaveTangent=-1.7004, tangentMode=RCTM_User, time=0.6, value=0.5}
          {arriveTangent=-0.4076, interpMode=RCIM_Cubic, leaveTangent=-0.4076, tangentMode=RCTM_User, time=1, value=0}
         ]}}}
      Scale Curve=1.5
     )))
   &GravityForce(Gravity=<0 0 60>)
   &Drag(Drag=0.5, Use Linear Drag=true, Use Rotational Drag=false)
   &ScaleColor(
    Scale Alpha=&FloatFromCurve(FloatCurve=DataInterface() {propertyValues {Curve {keys [<at -0.0046 1.0113 RCIM_Cubic>,<at 1.0011 0.002 RCIM_Cubic>]}}})
    Scale RGB=&VectorFromCurve(
     VectorCurve=DataInterface() {propertyValues {
       XCurve {keys [
         {arriveTangent=-1.8006, interpMode=RCIM_Cubic, leaveTangent=-1.8006, tangentMode=RCTM_User, time=0, value=1}
         {interpMode=RCIM_Cubic, leaveTangent=-0.8583, time=1, value=0.35}
        ]}
       YCurve {keys [
         {arriveTangent=-1.8006, interpMode=RCIM_Cubic, leaveTangent=-1.8006, tangentMode=RCTM_User, time=0, value=1}
         {interpMode=RCIM_Cubic, leaveTangent=-0.8583, time=1, value=0.35}
        ]}
       ZCurve {keys [
         {arriveTangent=-1.8006, interpMode=RCIM_Cubic, leaveTangent=-1.8006, tangentMode=RCTM_User, time=0, value=1}
         {interpMode=RCIM_Cubic, leaveTangent=-0.8583, time=1, value=0.35}
        ]}
      }}
     Scale Curve=&Multiply_VectorByFloat(Vector=<1 1 1>, Float=1.5)
    )
    ScaleA=false
    ScaleRGB=true
   )
   &ScaleColor(
    Color Value To Scale=$Particles.Color
    Scale RGB=&Lerp_Vector(
     A=&MakeVectorFromLinearColorRGB(Color=&Multiply_LinearColorByFloat(Linear Color=$NPC.NPC_Sparks.StageShowPrimary, Float=6))
     B=&MakeVectorFromLinearColorRGB(Color=&Multiply_LinearColorByFloat(Linear Color=$NPC.NPC_Sparks.StageShowTertiary, Float=1))
     Alpha=$StackContext.NormalizedAge
    )
    ScaleA=false
    ScaleRGB=true
   ) {module=ScaleColor002}
   &CurlNoiseForce(Bias Noise=false, Noise Frequency=500, Noise Strength=1000, Pan Noise=true, Pan Noise Field=<0 0 100>, Randomization Vector:Vector3f=<0 0 0>) {enabled=false}
   &VortexForce(UseOriginOffset=false) {enabled=false}
   &SolveForcesAndVelocity(Clamp Velocity=false, Limit Acceleration=false)
  ]
  rendererClasses [&NiagaraMeshRendererProperties]
  renderers [{rendererClass=&NiagaraMeshRendererProperties, rendererIndex=0}]
 }
 SmokeGround {
  emitterUpdateScript [
   &EmitterState()
   &SpawnRate(SpawnRate=0.5, Use Spawn Probability=false)
  ]
  particleSpawnScript [
   &InitializeParticle(
    Color=<rgba 0.1612 2 2 0.8>
    Lifetime=&RandomRangeFloat(Minimum=4, Maximum=6)
    Position Offset=<0 0 150>
    Sprite Size Mode=::NewEnumerator3
    Sprite UV Mode=::NewEnumerator3
    Uniform Sprite Size=&RandomRangeFloat(Minimum=1, Maximum=1.5)
    UsePositionOffset=true
    Write Material Random=true
   )
   &ShapeLocation(
    MaskSpawnGroup=false
    Offset=<0 0 60>
    Offset Mode=::NewEnumerator0
    Override Seed=false
    Ring Radius=200
    Shape Primitive=::NewEnumerator4
    UseOwnerScale=true
   )
   /SetVariables(Particles.SpriteRotation:NiagaraFloat=&RandomRangeFloat(Minimum=-0.33, Maximum=0.33)) {isSetParameters=true}
   &AddVelocity(Velocity=$StackContext.ShapeLocation.ShapeVector, Velocity Speed Scale=0.05)
   /SetVariables(Particles.SpriteSize=&Multiply_Vector2D_ByFloat(
     Float=&Lerp_Float(A=107.586, B=190.6381, Alpha=&MakeFloatFromInt(Int=$Particles.SpawnGroup))
     Vector2D=$Particles.SpriteSize
    )) {isSetParameters=true}
   &ApplyInitialForces(Positional Force Warmup Time=5)
  ]
  particleUpdateScript [
   &Drag(Drag=0.1, Use Linear Drag=true, Use Rotational Drag=true)
   &VortexForce(
    Origin Pull Amount=76.4139
    UseOriginOffset=false
    Vortex Axis=<1 0 0>
    Vortex Force Amount=&Lerp_Float(
     A=&RandomRangeFloat(Minimum=93.3802, Maximum=178.6482)
     B=0
     Alpha=&MakeFloatFromInt(Int=$Particles.SpawnGroup)
    )
   ) {enabled=false}
   &GravityForce(Gravity=&Lerp_Vector(A=<0 0 0>, B=<0 0 -226.7489>, Alpha=&MakeFloatFromInt(Int=$Particles.SpawnGroup))) {enabled=false}
   &AccelerationForce(Acceleration=<200 0 0>, Coordinate Space=::Local) {enabled=false}
   &SolveForcesAndVelocity(Clamp Velocity=false, Limit Acceleration=false)
   &ParticleState()
   &ScaleSpriteSize(
    Uniform Curve Scale=5
    Uniform Curve Sprite Scale=DataInterface() {propertyValues {Curve {keys [
        {arriveTangent=1.2245, interpMode=RCIM_Cubic, leaveTangent=1.2245, tangentMode=RCTM_User, time=0, value=1.5}
        {arriveTangent=0.05, interpMode=RCIM_Cubic, leaveTangent=0.05, time=0.9, value=2}
        <at 1 1.55 RCIM_Cubic>
       ]}}}
   )
   &ScaleColor(
    Scale RGB=&Lerp_Vector(
     A=&MakeVectorFromLinearColorRGB(Color=&Multiply_LinearColorByFloat(Linear Color=$NPC.NPC_Sparks.StageShowPrimary, Float=1))
     B=&MakeVectorFromLinearColorRGB(Color=<rgba 0.0434 0.0434 0.0434 1>)
     Alpha=&Multiply_Float(
      A=$Particles.NormalizedAge
      B=&FloatFromCurve(FloatCurve=DataInterface() {propertyValues {Curve {keys [
           {arriveTangent=3.1926, interpMode=RCIM_Cubic, leaveTangent=3.1926, tangentMode=RCTM_User, time=0, value=0}
           {arriveTangent=1, interpMode=RCIM_Cubic, leaveTangent=1, time=0.1, value=0.7}
           {interpMode=RCIM_Cubic, tangentMode=RCTM_User, time=1, value=1}
          ]}}})
     )
    )
    ScaleA=false
    ScaleRGB=true
   ) {module=ScaleColor001}
   /SetVariables(Particles.SubImageIndex=&FloatFromCurve(FloatCurve=DataInterface() {propertyValues {Curve {keys [<at 0 0>,<at 1 1>]}}}, Scale Curve=16)) {isSetParameters=true}
   &ScaleColor(
    Scale Alpha=&FloatFromCurve(FloatCurve=DataInterface() {propertyValues {Curve {keys [
         <at 0 0 RCIM_Cubic>
         {arriveTangent=0.5714, interpMode=RCIM_Cubic, leaveTangent=0.5714, time=0.2, value=0.775}
         {arriveTangent=-0.9688, interpMode=RCIM_Cubic, leaveTangent=-0.9688, time=0.7, value=0.4}
         <at 1 0 RCIM_Cubic>
        ]}}})
    Scale RGB=&VectorFromFloat(Value=&FloatFromCurve(FloatCurve=DataInterface() {propertyValues {Curve {keys [
          {arriveTangent=0.0071, interpMode=RCIM_Cubic, leaveTangent=0.0071, tangentMode=RCTM_User, time=0, value=1}
          {arriveTangent=-1, interpMode=RCIM_Cubic, leaveTangent=-1, time=0.3, value=0.425}
          {interpMode=RCIM_Cubic, tangentMode=RCTM_User, time=1, value=0}
         ]}}}))
    ScaleA=true
    ScaleRGB=true
   )
   &DynamicMaterialParameters(
    Index 0 Param 4=&FloatFromCurve(FloatCurve=DataInterface() {propertyValues {Curve {keys [<at 0 1>,<at 1 0>]}}}, Scale Curve=2)
    Module.Index0Param0Write=true
    Module.Index0Param1Write=true
    Module.Index0Param2Write=true
    Module.Index0Param3Write=true
   )
   &CameraOffset(Camera Offset Amount=&FloatFromCurve(FloatCurve=DataInterface() {propertyValues {Curve {keys [<at 0 -50>,<at 1 -10>]}}}))
   &self:MatchScale_02() {module=MatchScale_02}
  ]
  rendererClasses [&NiagaraSpriteRendererProperties]
  renderers [{rendererClass=&NiagaraSpriteRendererProperties}]
 }
}
summary {
 systemName=NS_Sparks_BattleStagePortalVFX
 emitters [
  {emitterName=SmokeGround, rendererClasses [&NiagaraSpriteRendererProperties]}
  {emitterName=GPU_Boxes001, simTarget=GPUComputeSim, rendererClasses [&NiagaraMeshRendererProperties]}
  {emitterName=Cheap_Boxes, rendererClasses [&NiagaraMeshRendererProperties]}
 ]
 userVariables []
}
